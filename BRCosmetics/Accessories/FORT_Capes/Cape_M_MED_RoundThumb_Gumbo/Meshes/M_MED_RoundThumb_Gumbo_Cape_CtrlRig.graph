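# Control Rig for the Gumbo cape cosmetic that applies pose-driven corrective offsets to upper arm and cape bones.
# Spherical pose readers on both upper arms weight self-relative transforms on added twist bones and the first left and right cape chain bones.
# Bool rc gates reader-driven cape corrections, while sky swaps two right-arm correctives.
# Bool sit eases in another right cape offset, scaled down by one reader's output.
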ControlRigBlueprint /BRCosmetics/Accessories/FORT_Capes/Cape_M_MED_RoundThumb_Gumbo/Meshes/M_MED_RoundThumb_Gumbo_Cape_CtrlRig
graphs [
 RigVMFunctionLibrary {links [], nodes []}
 RigVMModel {nodes [
   RigUnit_BeginExecution = &RigUnit_BeginExecution()
   RigVMFunction_Sequence = &RigVMFunction_Sequence(ExecuteContext=@RigUnit_BeginExecution.ExecutePin)
   n1 = &RigUnit_SphericalPoseReader(
    ActiveRegionScaleFactors=(PositiveWidth=2,NegativeWidth=3,PositiveHeight=3,NegativeHeight=3)
    ActiveRegionSize=0.05
    Debug=(bDrawDebug=false,DebugScale=25,DebugSegments=20,DebugThickness=0.25)
    DriverAxis=(X=1,Y=0,Z=0)
    DriverItem=(Type=Bone,Name="upperarm_l")
    FalloffRegionScaleFactors=(PositiveWidth=0.3,NegativeWidth=1,PositiveHeight=1,NegativeHeight=1)
    FalloffSize=0.4
    OptionalParentItem=(Type=Bone)
    RotationOffset=(X=28.192,Y=15.7826,Z=-38.8792)
    ExecutePin=@RigVMFunction_Sequence.A
   )
   > /Script/ControlRig.RigUnit_SetRelativeTransformForItem(
    Child=(Type=Bone,Name="upperarm_twist_01_l_add")
    Parent=(Type=Bone,Name="upperarm_twist_01_l_add")
    Value=(Rotation=(X=-0.1056,Y=0.4563,Z=-0.315,W=0.8254),Translation=(X=0,Y=0,Z=0),Scale3D=(X=1,Y=1,Z=1))
    bParentInitial=False
    bPropagateToChildren=True
    Weight=@n1.OutputParam
   ) {
    is=struct RigVMUnitNode
    template="Set Relative Transform::Execute(in Child,in Parent,in bParentInitial,in Value,in Weight,in bPropagateToChildren)"
   }
   > n2 = &RigUnit_SphericalPoseReader(
    ActiveRegionScaleFactors=(PositiveWidth=3,NegativeWidth=3,PositiveHeight=3,NegativeHeight=3)
    ActiveRegionSize=0.03
    Debug=(bDrawDebug=false,DebugScale=25,DebugSegments=20,DebugThickness=0.25)
    DriverAxis=(X=1,Y=0,Z=0)
    DriverItem=(Type=Bone,Name="upperarm_l")
    FalloffRegionScaleFactors=(PositiveWidth=1,NegativeWidth=1,PositiveHeight=1,NegativeHeight=1)
    FalloffSize=0.3
    OptionalParentItem=(Type=Bone)
    RotationOffset=(X=95.3502,Y=19.3239,Z=-19.0008)
   )
   > &RigUnit_SetRelativeTransformForItem(
    Child=(Type=Bone,Name="upperarm_twist_01_l_add")
    Parent=(Type=Bone,Name="upperarm_twist_01_l_add")
    Value=(Rotation=(X=-0.0722,Y=0.1111,Z=-0.7345,W=0.6656),Translation=(X=0,Y=0,Z=0),Scale3D=(X=1,Y=1,Z=1))
    bParentInitial=False
    bPropagateToChildren=True
    Weight=@n2.OutputParam
   ) {template="Set Relative Transform::Execute(in Child,in Parent,in bParentInitial,in Value,in Weight,in bPropagateToChildren)"}
   > n3 = &RigUnit_SphericalPoseReader(
    ActiveRegionScaleFactors=(PositiveWidth=2.5,NegativeWidth=3,PositiveHeight=3,NegativeHeight=3)
    ActiveRegionSize=0.03
    Debug=(bDrawDebug=false,DebugScale=25,DebugSegments=20,DebugThickness=0.25)
    DriverAxis=(X=1,Y=0,Z=0)
    DriverItem=(Type=Bone,Name="upperarm_l")
    FalloffRegionScaleFactors=(PositiveWidth=1,NegativeWidth=1,PositiveHeight=0.5,NegativeHeight=1)
    FalloffSize=0.3
    OptionalParentItem=(Type=Bone)
    RotationOffset=(X=-27.1958,Y=-48.6993,Z=44.8046)
   )
   > &RigUnit_SetRelativeTransformForItem(
    Child=(Type=Bone,Name="upperarm_twist_01_l_add")
    Parent=(Type=Bone,Name="upperarm_twist_01_l_add")
    Value=(Rotation=(X=0.0532,Y=0.4242,Z=-0.3151,W=0.8473),Translation=(X=0,Y=0,Z=0),Scale3D=(X=1,Y=1,Z=1))
    bParentInitial=False
    bPropagateToChildren=True
    Weight=@n3.OutputParam
   ) {template="Set Relative Transform::Execute(in Child,in Parent,in bParentInitial,in Value,in Weight,in bPropagateToChildren)"}
   > n4 = &RigUnit_SphericalPoseReader(
    ActiveRegionScaleFactors=(PositiveWidth=2.5,NegativeWidth=2.5,PositiveHeight=2.5,NegativeHeight=2.5)
    ActiveRegionSize=0.02
    Debug=(bDrawDebug=false,DebugScale=25,DebugSegments=20,DebugThickness=0.25)
    DriverAxis=(X=1,Y=0,Z=0)
    DriverItem=(Type=Bone,Name="upperarm_l")
    FalloffRegionScaleFactors=(PositiveWidth=2,NegativeWidth=2,PositiveHeight=2,NegativeHeight=2)
    FalloffSize=0.09
    OptionalParentItem=(Type=Bone)
    RotationOffset=(X=-33.136,Y=-90.0059,Z=0)
   )
   > &RigUnit_SetRelativeTransformForItem(
    Child=(Type=Bone,Name="dyn_cape_l_01")
    Parent=(Type=Bone,Name="dyn_cape_l_01")
    Value=(Rotation=(X=0.682,Y=-0.2146,Z=-0.0092,W=0.6991),Translation=(X=0,Y=0,Z=0),Scale3D=(X=1,Y=1,Z=1))
    bParentInitial=False
    bPropagateToChildren=True
    Weight=RigVMDispatchNode(False=0, True=@n4.OutputParam, Condition=RigVMVariableNode() {variable=rc}.Value) {template="DISPATCH_RigVMDispatch_If(in Condition,in True,in False,out Result)"}.Result
   ) {template="Set Relative Transform::Execute(in Child,in Parent,in bParentInitial,in Value,in Weight,in bPropagateToChildren)"}
   > n5 = &RigUnit_SphericalPoseReader(
    ActiveRegionScaleFactors=(PositiveWidth=3,NegativeWidth=3,PositiveHeight=4,NegativeHeight=3)
    ActiveRegionSize=0.07
    Debug=(bDrawDebug=false,DebugScale=25,DebugSegments=20,DebugThickness=0.25)
    DriverAxis=(X=-1,Y=0,Z=0)
    DriverItem=(Type=Bone,Name="upperarm_r")
    FalloffRegionScaleFactors=(PositiveWidth=0.5,NegativeWidth=0.5,PositiveHeight=1,NegativeHeight=0.5)
    FalloffSize=0.6
    FlipHeightScaling=true
    OptionalParentItem=(Type=Bone)
    RotationOffset=(X=137.8635,Y=-7.3017,Z=209.5155)
   )
   DISPATCH_RigVMDispatch_If = RigVMDispatchNode(True=0, False=@n5.OutputParam, Condition=RigVMVariableNode() {variable=sky}.Value) {template="DISPATCH_RigVMDispatch_If(in Condition,in True,in False,out Result)"}
   > &RigUnit_SetRelativeTransformForItem(
    Child=(Type=Bone,Name="upperarm_twist_01_add")
    Parent=(Type=Bone,Name="upperarm_twist_01_add")
    Value=(Rotation=(X=-0.2418,Y=0.1447,Z=-0.7931,W=0.5399),Translation=(X=0,Y=0,Z=0),Scale3D=(X=1,Y=1,Z=1))
    bParentInitial=False
    bPropagateToChildren=True
    Weight=@DISPATCH_RigVMDispatch_If.Result
   ) {template="Set Relative Transform::Execute(in Child,in Parent,in bParentInitial,in Value,in Weight,in bPropagateToChildren)"}
   > n6 = &RigUnit_SphericalPoseReader(
    ActiveRegionScaleFactors=(PositiveWidth=3,NegativeWidth=3,PositiveHeight=3,NegativeHeight=3)
    ActiveRegionSize=0.04
    Debug=(bDrawDebug=false,DebugScale=25,DebugSegments=20,DebugThickness=0.25)
    DriverAxis=(X=-1,Y=0,Z=0)
    DriverItem=(Type=Bone,Name="upperarm_r")
    FalloffRegionScaleFactors=(PositiveWidth=0.5,NegativeWidth=0.5,PositiveHeight=0.5,NegativeHeight=0.5)
    FalloffSize=0.4
    FlipHeightScaling=true
    OptionalParentItem=(Type=Bone)
    RotationOffset=(X=14.5139,Y=-43.6615,Z=232.5497)
   )
   > &RigUnit_SetRelativeTransformForItem(
    Child=(Type=Bone,Name="upperarm_twist_01_add")
    Parent=(Type=Bone,Name="upperarm_twist_01_add")
    Value=(Rotation=(X=0.1214,Y=0.055,Z=-0.7703,W=0.6236),Translation=(X=0,Y=0,Z=0),Scale3D=(X=1,Y=1,Z=1))
    bParentInitial=False
    bPropagateToChildren=True
    Weight=@n6.OutputParam
   ) {template="Set Relative Transform::Execute(in Child,in Parent,in bParentInitial,in Value,in Weight,in bPropagateToChildren)"}
   > n7 = &RigUnit_SphericalPoseReader(
    ActiveRegionScaleFactors=(PositiveWidth=2.5,NegativeWidth=3,PositiveHeight=3,NegativeHeight=3)
    ActiveRegionSize=0.03
    Debug=(bDrawDebug=false,DebugScale=25,DebugSegments=20,DebugThickness=0.25)
    DriverAxis=(X=-1,Y=0,Z=0)
    DriverItem=(Type=Bone,Name="upperarm_r")
    FalloffRegionScaleFactors=(PositiveWidth=1,NegativeWidth=1,PositiveHeight=1,NegativeHeight=1)
    FalloffSize=0.3
    FlipHeightScaling=true
    OptionalParentItem=(Type=Bone)
    RotationOffset=(X=175.4096,Y=48.6468,Z=180)
   )
   > &RigUnit_SetRelativeTransformForItem(
    Child=(Type=Bone,Name="upperarm_twist_01_add")
    Parent=(Type=Bone,Name="upperarm_twist_01_add")
    Value=(Rotation=(X=-0.0303,Y=0.3869,Z=-0.4963,W=0.7766),Translation=(X=-1.4575,Y=-0.7605,Z=-1.8346),Scale3D=(X=1,Y=1,Z=1))
    bParentInitial=False
    bPropagateToChildren=True
    Weight=RigVMDispatchNode(False=0, True=@n7.OutputParam, Condition=RigVMVariableNode() {variable=sky}.Value) {template="DISPATCH_RigVMDispatch_If(in Condition,in True,in False,out Result)"}.Result
   ) {template="Set Relative Transform::Execute(in Child,in Parent,in bParentInitial,in Value,in Weight,in bPropagateToChildren)"}
   > n8 = &RigUnit_SphericalPoseReader(
    ActiveRegionScaleFactors=(PositiveWidth=2.5,NegativeWidth=2.5,PositiveHeight=2.5,NegativeHeight=2.5)
    ActiveRegionSize=0.015
    Debug=(bDrawDebug=false,bDraw2D=False,bDrawLocalAxes=False,DebugScale=25,DebugSegments=20,DebugThickness=0.25)
    DriverAxis=(X=-1,Y=0,Z=0)
    DriverItem=(Type=Bone,Name="upperarm_r")
    FalloffRegionScaleFactors=(PositiveWidth=3,NegativeWidth=3,PositiveHeight=3,NegativeHeight=3)
    FalloffSize=0.06
    FlipHeightScaling=False
    FlipWidthScaling=False
    OptionalParentItem=(Type=Bone,Name="None")
    RotationOffset=(X=-206.1819,Y=-115.618,Z=2.1015)
   )
   > RigUnit_SetRelativeTransformForItem = &RigUnit_SetRelativeTransformForItem(
    Child=(Type=Bone,Name="dyn_cape_r_01")
    Parent=(Type=Bone,Name="dyn_cape_r_01")
    Value=(Rotation=(X=0,Y=-0.1213,Z=0,W=0.9926),Translation=(X=0,Y=0,Z=0),Scale3D=(X=1,Y=1,Z=1))
    bParentInitial=False
    bPropagateToChildren=True
    Weight=RigVMDispatchNode(False=0, True=@n8.OutputParam, Condition=RigVMVariableNode() {variable=rc}.Value) {template="DISPATCH_RigVMDispatch_If(in Condition,in True,in False,out Result)"}.Result
   ) {template="Set Relative Transform::Execute(in Child,in Parent,in bParentInitial,in Value,in Weight,in bPropagateToChildren)"}
   RigVMFunction_MathBoolToFloat = &RigVMFunction_MathBoolToFloat(Value=RigVMVariableNode() {variable=sit}.Value)
   AlphaInterp = &RigVMFunction_AlphaInterp(InterpSpeedDecreasing=10, InterpSpeedIncreasing=5, Scale=1, bClampResult=False, Value=@RigVMFunction_MathBoolToFloat.Result)
   Subtract = &RigVMFunction_MathDoubleSub(A=1, B=@n8.OutputParam)
   Multiply = &RigVMFunction_MathDoubleMul(B=@AlphaInterp.Result, A=@Subtract.Result)
   > &RigUnit_SetRelativeTransformForItem(
    Child=(Type=Bone,Name="dyn_cape_r_01")
    Parent=(Type=Bone,Name="dyn_cape_r_01")
    Value=(Rotation=(X=0.0123,Y=0.1058,Z=0.0757,W=0.9914),Translation=(X=0.4587,Y=-1.3023,Z=1.2725),Scale3D=(X=1,Y=1,Z=1))
    bParentInitial=False
    bPropagateToChildren=True
    Weight=@Multiply.Result
   ) {template="Set Relative Transform::Execute(in Child,in Parent,in bParentInitial,in Value,in Weight,in bPropagateToChildren)"}
   > n9 = &RigUnit_SphericalPoseReader(
    ActiveRegionScaleFactors=(PositiveWidth=2.5,NegativeWidth=2.5,PositiveHeight=2.5,NegativeHeight=2.5)
    ActiveRegionSize=0.02
    Debug=(bDrawDebug=false,bDraw2D=False,bDrawLocalAxes=False,DebugScale=25,DebugSegments=20,DebugThickness=0.25)
    DriverAxis=(X=-1,Y=0,Z=0)
    DriverItem=(Type=Bone,Name="upperarm_r")
    FalloffRegionScaleFactors=(PositiveWidth=2,NegativeWidth=2,PositiveHeight=2,NegativeHeight=2)
    FalloffSize=0.09
    FlipHeightScaling=False
    FlipWidthScaling=False
    OptionalParentItem=(Type=Bone,Name="None")
    RotationOffset=(X=-278.3698,Y=-161.118,Z=63.9168)
   )
   > &RigUnit_SetRelativeTransformForItem(
    Child=(Type=Bone,Name="dyn_cape_r_01")
    Parent=(Type=Bone,Name="dyn_cape_r_01")
    Value=(Rotation=(X=-0.487,Y=-0.195,Z=-0.0478,W=0.85),Translation=(X=0,Y=0,Z=0),Scale3D=(X=1,Y=1,Z=1))
    bParentInitial=False
    bPropagateToChildren=True
    Weight=RigVMDispatchNode(False=0, True=@n9.OutputParam, Condition=RigVMVariableNode() {variable=rc}.Value) {template="DISPATCH_RigVMDispatch_If(in Condition,in True,in False,out Result)"}.Result
   ) {template="Set Relative Transform::Execute(in Child,in Parent,in bParentInitial,in Value,in Weight,in bPropagateToChildren)"}
  ]}
]
hierarchy [
 root {loc=<0 0 0>, rot=<0 0 0 1>}
 pelvis {parent=root, loc=<0 -0.7071 93.1191>, rot=<0.0127 0.707 -0.0127 -0.707>}
 spine_01 {parent=pelvis, loc=<6.7494 -0.6388 0>, rot=<0 0 0.034 -0.9994>}
 spine_02 {parent=spine_01, loc=<7.2518 -0.035 0>, rot=<0 0 0.0238 -0.9997>}
 spine_03 {parent=spine_02, loc=<6.0104 0 0>, rot=<0 0 0.0149 -0.9999>}
 spine_04 {parent=spine_03, loc=<8.1024 0 0>, rot=<0 0 -0.0132 -0.9999>}
 spine_05 {parent=spine_04, loc=<10.0146 0 0>, rot=<0 0 -0.0015 -1>}
 clavicle_l {parent=spine_05, loc=<4.3575 -0.1466 -2.3223>, rot=<0.0465 -0.7289 -0.009 -0.6829>}
 upperarm_l {parent=clavicle_l, loc=<15.5111 0 0>, rot=<0.0226 -0.3234 -0.0088 -0.9459>}
 upperarm_twist_01_l {parent=upperarm_l, loc=<0.0041 0 0>, rot=<0 0 0 1>}
 deform_upperarm_bk_base_l {parent=upperarm_twist_01_l, loc=<1.0813 -7.2193 -1.9453>, rot=<0 0 0 1>}
 deform_upperarm_bk_l {parent=deform_upperarm_bk_base_l, loc=<0 0 0>, rot=<0 0 0 1>}
 deform_upperarm_fr_base_l {parent=upperarm_twist_01_l, loc=<-0.3691 7.0506 -2.765>, rot=<0 0 0 1>}
 deform_upperarm_fr_l {parent=deform_upperarm_fr_base_l, loc=<0 0 0>, rot=<0 0 0 1>}
 deform_upperarm_up_base_l {parent=upperarm_twist_01_l, loc=<-3.2216 -1.1391 3.3311>, rot=<0 0 0 1>}
 deform_upperarm_up_l {parent=deform_upperarm_up_base_l, loc=<0 0 0>, rot=<0 0 0 1>}
 upperarm_twist_01_l_add {parent=upperarm_twist_01_l, loc=<0 0 0>, rot=<0 0 0 1>}
 upperarm_twist_02_l {parent=upperarm_l, loc=<13.7941 0 0>, rot=<0 0 0 1>}
 lowerarm_l {parent=upperarm_l, loc=<26.6514 0 0>, rot=<0.0391 -0.0099 -0.2453 -0.9686>}
 Kallar_L {parent=clavicle_l, loc=<12.3543 -0.3308 7.9224>, rot=<-0.0141 0.6135 -0.0209 -0.7893>}
 clavicle_r {parent=spine_05, loc=<4.3575 -0.1466 2.3223>, rot=<0.7289 0.0465 -0.6829 0.009>}
 upperarm_r {parent=clavicle_r, loc=<-15.5111 0 0>, rot=<0.0226 -0.3234 -0.0088 -0.9459>}
 upperarm_twist_01_r {parent=upperarm_r, loc=<-0.0041 0 0>, rot=<0 0 0 -1>}
 deform_upperarm_bk_base_r {parent=upperarm_twist_01_r, loc=<-1.0814 7.2192 1.9456>, rot=<0 0 0 -1>}
 deform_upperarm_bk_r {parent=deform_upperarm_bk_base_r, loc=<0 0 0>, rot=<0 0 0 1>}
 deform_upperarm_fr_base_r {parent=upperarm_twist_01_r, loc=<0.3689 -7.0506 2.7652>, rot=<0 0 0 -1>}
 deform_upperarm_fr_r {parent=deform_upperarm_fr_base_r, loc=<0 0 0>, rot=<0 0 0 1>}
 deform_upperarm_up_base_r {parent=upperarm_twist_01_r, loc=<3.2214 1.1391 -3.3308>, rot=<0 0 0 -1>}
 deform_upperarm_up_r {parent=deform_upperarm_up_base_r, loc=<0 0 0>, rot=<0 0 0 -1>}
 upperarm_twist_01_add {parent=upperarm_twist_01_r, loc=<0 0 0>, rot=<0 0 0 -1>}
 upperarm_twist_02_r {parent=upperarm_r, loc=<-13.7941 0 0>, rot=<0 0 0 -1>}
 lowerarm_r {parent=upperarm_r, loc=<-26.6514 0 0>, rot=<0.0391 -0.0099 -0.2453 -0.9686>}
 Kallar_R {parent=clavicle_r, loc=<-12.3544 0.3308 -7.9219>, rot=<-0.0141 0.6135 -0.0209 -0.7893>}
 backpack_root {parent=spine_05, loc=<10.4135 -12.3103 0>, rot=<0.0284 -0.7065 -0.0284 -0.7065>}
 hider {parent=backpack_root, loc=<0 0.1621 0.0076>, rot=<0 0 0 1>}
 cape_top {parent=hider, loc=<0 0.0288 -2.7904>, rot=<0 0 0 1>}
 dyn_cape_r_MSTR {parent=cape_top, loc=<-21.5 1.5541 -2.5734>, rot=<-0.4358 -0.4358 0.5569 -0.5569>}
 dyn_cape_r_01 {parent=dyn_cape_r_MSTR, loc=<0 0 0>, rot=<0 0 0 1>}
 dyn_cape_r_02 {parent=dyn_cape_r_01, loc=<17.2291 0 -0.1468>, rot=<0 -0.0648 0 -0.9979>}
 dyn_cape_r_03 {parent=dyn_cape_r_02, loc=<17.2381 0 -0.0337>, rot=<0 -0.0149 0 -0.9999>}
 dyn_cape_r_04 {parent=dyn_cape_r_03, loc=<17.2387 0 -0.0104>, rot=<0 -0.0046 0 -1>}
 dyn_cape_r_05 {parent=dyn_cape_r_04, loc=<17.2387 0 0>, rot=<0 0 0 1>}
 dyn_cape_r_06 {parent=dyn_cape_r_05, loc=<17.2386 0 0>, rot=<0 0 0 1>}
 dyn_cape_r_07 {parent=dyn_cape_r_06, loc=<17.2385 0 0.0001>, rot=<0 0 0 1>}
 dyn_cape_l_MSTR {parent=cape_top, loc=<21.5 1.5541 -2.5734>, rot=<-0.4358 -0.4358 0.5569 -0.5569>}
 dyn_cape_l_01 {parent=dyn_cape_l_MSTR, loc=<0 0 0>, rot=<0 0 0 1>}
 dyn_cape_l_02 {parent=dyn_cape_l_01, loc=<17.2291 0 -0.1468>, rot=<0 -0.0648 0 -0.9979>}
 dyn_cape_l_03 {parent=dyn_cape_l_02, loc=<17.2382 0 -0.0337>, rot=<0 -0.0149 0 -0.9999>}
 dyn_cape_l_04 {parent=dyn_cape_l_03, loc=<17.2386 0 -0.0104>, rot=<0 -0.0046 0 -1>}
 dyn_cape_l_05 {parent=dyn_cape_l_04, loc=<17.2387 0 0>, rot=<0 0 0 1>}
 dyn_cape_l_06 {parent=dyn_cape_l_05, loc=<17.2387 0 0>, rot=<0 0 0 1>}
 dyn_cape_l_07 {parent=dyn_cape_l_06, loc=<17.2387 0 0>, rot=<0 0 0 1>}
 dyn_cape_MSTR {parent=cape_top, loc=<0 -0.0317 -2.5734>, rot=<-0.4358 -0.4358 0.5569 -0.5569>}
 dyn_cape_01 {parent=dyn_cape_MSTR, loc=<0 0 0>, rot=<0 0 0 1>}
 dyn_cape_02 {parent=dyn_cape_01, loc=<17.2291 0 -0.1468>, rot=<0 -0.0648 0 -0.9979>}
 dyn_cape_03 {parent=dyn_cape_02, loc=<17.2382 0 -0.0337>, rot=<0 -0.0149 0 -0.9999>}
 dyn_cape_04 {parent=dyn_cape_03, loc=<17.2386 0 -0.0104>, rot=<0 -0.0046 0 -1>}
 dyn_cape_05 {parent=dyn_cape_04, loc=<17.2387 0 0>, rot=<0 0 0 1>}
 dyn_cape_06 {parent=dyn_cape_05, loc=<17.2387 0 0>, rot=<0 0 0 1>}
 dyn_cape_07 {parent=dyn_cape_06, loc=<17.2387 0 0>, rot=<0 0 0 1>}
 correct_top_cape_jnt_01_l {parent=spine_05, loc=<11.7743 -3.7151 -20.5969>, rot=<0.7065 0.0284 -0.7065 0.0284>}
 correct_top_cape_jnt_02_l {parent=spine_05, loc=<7.3323 -11.8256 -19.7287>, rot=<0.7065 0.0284 -0.7065 0.0284>}
 correct_top_cape_jnt_03_l {parent=spine_05, loc=<2.7797 -14.3857 -12.7407>, rot=<0.7065 0.0284 -0.7065 0.0284>}
 correct_top_cape_jnt_m {parent=spine_05, loc=<0.2327 -16.9191 0.0729>, rot=<0.7065 0.0284 -0.7065 0.0284>}
 correct_top_cape_jnt_01_r {parent=spine_05, loc=<11.7746 -3.7151 20.5969>, rot=<0.0284 -0.7065 -0.0284 -0.7065>}
 correct_top_cape_jnt_02_r {parent=spine_05, loc=<7.332 -11.826 19.729>, rot=<0.0284 -0.7065 -0.0284 -0.7065>}
 correct_top_cape_jnt_03_r {parent=spine_05, loc=<2.7793 -14.3858 12.7407>, rot=<0.0284 -0.7065 -0.0284 -0.7065>}
 correct_clv_joint_r {parent=spine_05, loc=<2.7793 -14.3858 12.7407>, rot=<0.0284 -0.7065 -0.0284 -0.7065>}
 correct_clv_joint_l {parent=spine_05, loc=<2.7797 -14.3857 -12.7407>, rot=<0.7065 0.0284 -0.7065 0.0284>}
 attach_cape {parent=spine_05, loc=<9.0171 -11.714 0>, rot=<0.0616 -0.7045 -0.0615 -0.7043>}
 attach_backpack {parent=spine_05, loc=<-0.1229 -13.473 0>, rot=<-0.0093 -0.7071 0.0094 -0.707>}
 neck_01 {parent=spine_05, loc=<11.3138 0.1529 0>, rot=<0.1015 -0.0116 -0.1131 -0.9883>}
 neck_02 {parent=neck_01, loc=<4.5627 0 0>, rot=<-0.0135 -0.0034 0.0565 -0.9983>}
 head {parent=neck_02, loc=<5.3914 0 0>, rot=<0.048 0.0194 0.0742 -0.9959>}
 deform_pec_aim_base_l {parent=spine_04, loc=<7.8732 12.6388 -2.4123>, rot=<0.0102 -0.4602 0.3979 -0.7936>}
 deform_pec_aim_l {parent=deform_pec_aim_base_l, loc=<0 0 0>, rot=<0 0 0 1>}
 deform_pec_base_l {parent=deform_pec_aim_l, loc=<10.766 0 0>, rot=<0 0 0 1>}
 deform_pec_l {parent=deform_pec_base_l, loc=<0.055 0 0>, rot=<0 0 0 1>}
 deform_pec_aim_base_r {parent=spine_04, loc=<7.8727 12.6388 2.4123>, rot=<-0.4602 -0.0102 0.7936 0.3979>}
 deform_pec_aim_r {parent=deform_pec_aim_base_r, loc=<0 0 0>, rot=<0 0 0 1>}
 deform_pec_base_r {parent=deform_pec_aim_r, loc=<-10.7662 -0.0006 -0.0007>, rot=<0 0 0 1>}
 deform_pec_r {parent=deform_pec_base_r, loc=<0 0 0>, rot=<0 0 0 1>}
 thigh_l {parent=pelvis, loc=<-2.5693 -0.0109 -7.7501>, rot=<0.0555 0.111 0.2051 -0.9708>}
 calf_l {parent=thigh_l, loc=<-38.9834 0.0124 -0.0015>, rot=<-0.0003 -0.0169 -0.1892 -0.9818>}
 foot_l {parent=calf_l, loc=<-42.561 0.1077 0.4292>, rot=<0.0191 -0.0868 -0.0146 -0.9959>}
 thigh_r {parent=pelvis, loc=<-2.5691 -0.0109 7.7501>, rot=<0.1362 -0.1844 0.9732 0.0203>}
 calf_r {parent=thigh_r, loc=<38.984 -0.0122 0.0014>, rot=<-0.0002 -0.016 -0.1795 -0.9836>}
 foot_r {parent=calf_r, loc=<42.5612 -0.1077 -0.4292>, rot=<0.0153 -0.1279 0.1357 -0.9823>}
]
variables [
 rc {type=bool, default=False}
 sit {type=bool, default=False}
 sky {type=bool, default=False}
]
